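# Physics asset for the female medium Innovator Plaid body cosmetic, set up for secondary motion of its necktie.
# A chain of box bodies runs from a small sphere at the tie base through five segments, joined by constraints.
# Swing limits widen from the base toward the tip, while twist stays tightly limited throughout.
# Capsules on spine_03 and spine_05 approximate the torso, apparently as collision for the tie.
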PhysicsAsset /BRCosmetics/Characters/Player/Female/Medium/Bodies/F_MED_Innovator_Plaid/Meshes/F_MED_Innovator_Plaid_Physics
bodies {
 dyn_tie_01 [{shapeType=Box, extentX=5.1682, extentY=7.7415, extentZ=1.5, center=<2.5424 0 0>}]
 dyn_tie_02 [{shapeType=Box, extentX=4.9914, extentY=7.8376, extentZ=1.4344, center=<2.6138 0 0>}]
 dyn_tie_03 [{shapeType=Box, extentX=4.5868, extentY=9.305, extentZ=1.5242, center=<2.4452 0 0>}]
 dyn_tie_04 [{shapeType=Box, extentX=6.76, extentY=10.76, extentZ=1.25, center=<2.853 0 0>}]
 dyn_tie_base [{shapeType=Sphere, radius=0.5}]
 dyn_tie_parent [{shapeType=Box, extentX=5.9812, extentY=7.7843, extentZ=1.75, center=<3.2234 0 -0.0004>}]
 spine_03 [{shapeType=Capsule, length=2.75, radius=10.125, center=<5.1655 0.724 -0.0001>, rotation=<rot 90 0 -6.0208>}]
 spine_05 [
  {shapeType=Capsule, length=10.75, radius=6.875, center=<-3.0905 3.3678 0.0001>}
  {
   shapeType=Capsule
   length=12.5
   radius=10.25
   shapeName=spine_05_capsule
   center=<-0.2059 -1.9284 0.0001>
   rotation=<rot 90 -10.6197 15.6648>
  }
 ]
}
constraints [
 {bone1Name=dyn_tie_parent, bone2Name=dyn_tie_base, swing1LimitDegrees=10, swing2LimitDegrees=10, twistLimitDegrees=3}
 {bone1Name=dyn_tie_01, bone2Name=dyn_tie_parent, swing1LimitDegrees=15, swing2LimitDegrees=15, twistLimitDegrees=3}
 {bone1Name=dyn_tie_02, bone2Name=dyn_tie_01, swing1LimitDegrees=20, swing2LimitDegrees=20, twistLimitDegrees=3}
 {bone1Name=dyn_tie_03, bone2Name=dyn_tie_02, swing1LimitDegrees=20, swing2LimitDegrees=20, twistLimitDegrees=3}
 {bone1Name=dyn_tie_04, bone2Name=dyn_tie_03, swing1LimitDegrees=20, swing2LimitDegrees=20, twistLimitDegrees=3}
]
