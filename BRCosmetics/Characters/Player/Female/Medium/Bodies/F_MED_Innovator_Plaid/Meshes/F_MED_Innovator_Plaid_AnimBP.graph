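# Post-mesh AnimBlueprint for the female medium Innovator Plaid body cosmetic that adds secondary physics motion to its clothing.
# It copies the pose from the attached parent mesh, then chains Anim Dynamics nodes for sleeve, pant and boot zipper bones.
# Sleeves and pants use springs at reduced alpha, while zippers swing within angular limits.
# A final RigidBody node simulates the outfit's physics asset in pelvis space; the event graph has no logic.
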AnimBlueprint /BRCosmetics/Characters/Player/Female/Medium/Bodies/F_MED_Innovator_Plaid/Meshes/F_MED_Innovator_Plaid_AnimBP
target_skeleton=&/F_MED_Innovator_Plaid_Skeleton
graphs [
 AnimGraph {nodes [
   n1 = Anim Dynamics - Bone: dyn_sleeve_inner_r(ComponentPose=Local To Component(LocalPose=Copy Pose From Mesh() {bUseAttachedParent=true}), Alpha=1, ExternalForce="0,0,0") {
    angularDampingOverride=0.9999
    angularSpringConstant=5000
    bAngularSpring=true
    bGravityOverrideInSimSpace=true
    bOverrideAngularDamping=true
    gravityScale=0.35
    linearDampingOverride=0.9999
    linearSpringConstant=1000
    alphaBoolBlend {blendOption=Linear}
    alphaScaleBiasClamp {scale=0.35}
    boundBone {boneName=dyn_sleeve_inner_r}
    componentAppliedLinearAccClamp=<6000 6000 6000>
    componentLinearAccScale=<0.25 0.25 0.5>
    componentLinearVelScale=<0.25 0.25 0.5>
    physicsBodyDefinitions [{
      sphereCollisionRadius=0
      boundBone {boneName=dyn_sleeve_inner_r}
      boxExtents=<8 25 25>
      constraintSetup {angularLimitsMax=<45 20 20>, angularLimitsMin=<-45 -20 -75>, angularTarget=<1 0 0>}
      localJointOffset=<5 0 0>
     }]
    relativeSpaceBone {boneName=pelvis}
   }
   n2 = Anim Dynamics - Bone: dyn_sleeve_outter_r(ComponentPose=@n1, Alpha=1, ExternalForce="0,0,0") {
    angularDampingOverride=0.9999
    angularSpringConstant=5000
    bAngularSpring=true
    bGravityOverrideInSimSpace=true
    bOverrideAngularDamping=true
    gravityScale=0.35
    linearDampingOverride=0.9999
    linearSpringConstant=1000
    alphaBoolBlend {blendOption=Linear}
    alphaScaleBiasClamp {scale=0.35}
    boundBone {boneName=dyn_sleeve_outter_r}
    componentAppliedLinearAccClamp=<6000 6000 6000>
    componentLinearAccScale=<0.25 0.25 0.5>
    componentLinearVelScale=<0.25 0.25 0.5>
    physicsBodyDefinitions [{
      sphereCollisionRadius=0
      boundBone {boneName=dyn_sleeve_outter_r}
      boxExtents=<8 25 25>
      constraintSetup {angularLimitsMax=<45 40 20>, angularLimitsMin=<-45 -20 -75>, angularTarget=<1 0 0>}
      localJointOffset=<5 0 0>
     }]
    relativeSpaceBone {boneName=pelvis}
   }
   n3 = Anim Dynamics - Bone: dyn_sleeve_outter_l(ComponentPose=@n2, Alpha=1, ExternalForce="0,0,0") {
    angularDampingOverride=0.9999
    angularSpringConstant=5000
    bAngularSpring=true
    bGravityOverrideInSimSpace=true
    bOverrideAngularDamping=true
    gravityScale=0.35
    linearDampingOverride=0.9999
    linearSpringConstant=1000
    alphaBoolBlend {blendOption=Linear}
    alphaScaleBiasClamp {scale=0.35}
    boundBone {boneName=dyn_sleeve_outter_l}
    componentAppliedLinearAccClamp=<6000 6000 6000>
    componentLinearAccScale=<0.25 0.25 0.5>
    componentLinearVelScale=<0.25 0.25 0.5>
    physicsBodyDefinitions [{
      sphereCollisionRadius=0
      boundBone {boneName=dyn_sleeve_outter_l}
      boxExtents=<8 25 25>
      constraintSetup {angularLimitsMax=<45 40 20>, angularLimitsMin=<-45 -20 -75>, angularTarget=<1 0 0>}
      localJointOffset=<5 0 0>
     }]
    relativeSpaceBone {boneName=pelvis}
   }
   n4 = Anim Dynamics - Bone: dyn_sleeve_inner_l(ComponentPose=@n3, Alpha=1, ExternalForce="0,0,0") {
    angularDampingOverride=0.9999
    angularSpringConstant=5000
    bAngularSpring=true
    bGravityOverrideInSimSpace=true
    bOverrideAngularDamping=true
    gravityScale=0.35
    linearDampingOverride=0.9999
    linearSpringConstant=1000
    alphaBoolBlend {blendOption=Linear}
    alphaScaleBiasClamp {scale=0.35}
    boundBone {boneName=dyn_sleeve_inner_l}
    componentAppliedLinearAccClamp=<6000 6000 6000>
    componentLinearAccScale=<0.25 0.25 0.5>
    componentLinearVelScale=<0.25 0.25 0.5>
    physicsBodyDefinitions [{
      sphereCollisionRadius=0
      boundBone {boneName=dyn_sleeve_inner_l}
      boxExtents=<8 25 25>
      constraintSetup {angularLimitsMax=<45 20 20>, angularLimitsMin=<-45 -20 -75>, angularTarget=<1 0 0>}
      localJointOffset=<5 0 0>
     }]
    relativeSpaceBone {boneName=pelvis}
   }
   n5 = Anim Dynamics - Bone: dyn_pants_inner_r(ComponentPose=@n4, Alpha=1, ExternalForce="0,0,0") {
    angularDampingOverride=0.9999
    angularSpringConstant=1500
    bAngularSpring=true
    bGravityOverrideInSimSpace=true
    bLinearSpring=true
    bOverrideAngularDamping=true
    bOverrideLinearDamping=true
    gravityScale=0.75
    linearDampingOverride=0.9999
    linearSpringConstant=1500
    simulationSpace=RootRelative
    alphaBoolBlend {blendOption=Linear}
    alphaScaleBiasClamp {scale=0.25}
    boundBone {boneName=dyn_pants_inner_r}
    componentAppliedLinearAccClamp=<6000 6000 6000>
    componentLinearAccScale=<0.25 0.25 0.5>
    componentLinearVelScale=<0.25 0.25 0.5>
    physicsBodyDefinitions [{
      sphereCollisionRadius=0
      boundBone {boneName=dyn_pants_inner_r}
      boxExtents=<8 25 25>
      constraintSetup {angularLimitsMax=<60 40 40>, angularLimitsMin=<-60 -40 -75>, angularTarget=<1 0 0>}
      localJointOffset=<5 0 0>
     }]
    relativeSpaceBone {boneName=root}
   }
   n6 = Anim Dynamics - Bone: dyn_pants_outter_r(ComponentPose=@n5, Alpha=1, ExternalForce="0,0,0") {
    angularDampingOverride=0.9999
    angularSpringConstant=1500
    bAngularSpring=true
    bGravityOverrideInSimSpace=true
    bLinearSpring=true
    bOverrideAngularDamping=true
    bOverrideLinearDamping=true
    gravityScale=0.75
    linearDampingOverride=0.9999
    linearSpringConstant=1500
    simulationSpace=RootRelative
    alphaBoolBlend {blendOption=Linear}
    alphaScaleBiasClamp {scale=0.25}
    boundBone {boneName=dyn_pants_outter_r}
    componentAppliedLinearAccClamp=<6000 6000 6000>
    componentLinearAccScale=<0.25 0.25 0.5>
    componentLinearVelScale=<0.25 0.25 0.5>
    physicsBodyDefinitions [{
      sphereCollisionRadius=0
      boundBone {boneName=dyn_pants_outter_r}
      boxExtents=<8 25 25>
      constraintSetup {angularLimitsMax=<60 40 40>, angularLimitsMin=<-60 -40 -75>, angularTarget=<1 0 0>}
      localJointOffset=<5 0 0>
     }]
    relativeSpaceBone {boneName=root}
   }
   n7 = Anim Dynamics - Bone: dyn_pants_outter_l(ComponentPose=@n6, Alpha=1, ExternalForce="0,0,0") {
    angularDampingOverride=0.9999
    angularSpringConstant=2000
    bAngularSpring=true
    bGravityOverrideInSimSpace=true
    bLinearSpring=true
    bOverrideAngularDamping=true
    bOverrideLinearDamping=true
    gravityScale=0.75
    linearDampingOverride=0.9999
    linearSpringConstant=2000
    simulationSpace=RootRelative
    alphaBoolBlend {blendOption=Linear}
    alphaScaleBiasClamp {scale=0.25}
    boundBone {boneName=dyn_pants_outter_l}
    componentAppliedLinearAccClamp=<6000 6000 6000>
    componentLinearAccScale=<0.25 0.25 0.5>
    componentLinearVelScale=<0.25 0.25 0.5>
    physicsBodyDefinitions [{
      sphereCollisionRadius=0
      boundBone {boneName=dyn_pants_outter_l}
      boxExtents=<8 25 25>
      constraintSetup {angularLimitsMax=<60 40 40>, angularLimitsMin=<-60 -40 -75>, angularTarget=<1 0 0>}
      localJointOffset=<5 0 0>
     }]
    relativeSpaceBone {boneName=root}
   }
   n8 = Anim Dynamics - Bone: dyn_pants_inner_l(ComponentPose=@n7, Alpha=1, ExternalForce="0,0,0") {
    angularDampingOverride=0.9999
    angularSpringConstant=2000
    bAngularSpring=true
    bGravityOverrideInSimSpace=true
    bLinearSpring=true
    bOverrideAngularDamping=true
    bOverrideLinearDamping=true
    gravityScale=0.75
    linearDampingOverride=0.9999
    linearSpringConstant=2000
    simulationSpace=RootRelative
    alphaBoolBlend {blendOption=Linear}
    alphaScaleBiasClamp {scale=0.25}
    boundBone {boneName=dyn_pants_inner_l}
    componentAppliedLinearAccClamp=<6000 6000 6000>
    componentLinearAccScale=<0.25 0.25 0.5>
    componentLinearVelScale=<0.25 0.25 0.5>
    physicsBodyDefinitions [{
      sphereCollisionRadius=0
      boundBone {boneName=dyn_pants_inner_l}
      boxExtents=<8 25 25>
      constraintSetup {angularLimitsMax=<60 40 40>, angularLimitsMin=<-60 -40 -75>, angularTarget=<1 0 0>}
      localJointOffset=<5 0 0>
     }]
    relativeSpaceBone {boneName=root}
   }
   n9 = Anim Dynamics - Bone: dyn_boot_zipper_parent_r(ComponentPose=@n8, Alpha=1, ExternalForce="0,0,0") {
    angularDampingOverride=0.9999
    angularSpringConstant=0.5
    bGravityOverrideInSimSpace=true
    bOverrideAngularDamping=true
    bOverrideLinearDamping=true
    linearDampingOverride=0.9999
    simulationSpace=RootRelative
    alphaBoolBlend {blendOption=Linear}
    boundBone {boneName=dyn_boot_zipper_parent_r}
    componentAppliedLinearAccClamp=<10000 10000 10000>
    componentLinearAccScale=<0.15 0.15 0.15>
    componentLinearVelScale=<0.15 0.15 0.15>
    gravityOverride=<1 1 1>
    physicsBodyDefinitions [{
      sphereCollisionRadius=0
      boundBone {boneName=dyn_boot_zipper_parent_r}
      boxExtents=<6 6 2>
      constraintSetup {angularLimitsMax=<0 40 30>, angularLimitsMin=<0 -2 -30>, angularTarget=<0 0 0>}
      localJointOffset=<2.5 0 0>
     }]
    relativeSpaceBone {boneName=root}
   }
   n10 = Anim Dynamics - Bone: dyn_boot_zipper_parent_l(ComponentPose=@n9, Alpha=1, ExternalForce="0,0,0") {
    angularDampingOverride=0.9999
    angularSpringConstant=0.99
    bGravityOverrideInSimSpace=true
    bOverrideAngularDamping=true
    bOverrideLinearDamping=true
    linearDampingOverride=0.9999
    simulationSpace=RootRelative
    alphaBoolBlend {blendOption=Linear}
    boundBone {boneName=dyn_boot_zipper_parent_l}
    componentAppliedLinearAccClamp=<10000 10000 10000>
    componentLinearAccScale=<0.15 0.15 0.15>
    componentLinearVelScale=<0.15 0.15 0.15>
    gravityOverride=<1 1 1>
    physicsBodyDefinitions [{
      sphereCollisionRadius=0
      boundBone {boneName=dyn_boot_zipper_parent_l}
      boxExtents=<6 6 2>
      constraintSetup {angularLimitsMax=<0 40 30>, angularLimitsMin=<0 -2 -30>, angularTarget=<0 0 0>}
      localJointOffset=<2.5 0 0>
     }]
    relativeSpaceBone {boneName=root}
   }
   n11 = RigidBody(ComponentPose=@n10) {
    bClampLinearTranslationLimitToRefPose=true
    bOverrideWorldGravity=true
    overridePhysicsAsset=&/F_MED_Innovator_Plaid_Physics
    simulationSpace=BaseBoneSpace
    alphaBoolBlend {blendOption=Linear}
    baseBoneRef {boneName=pelvis}
    componentAppliedLinearAccClamp=<6000 6000 6000>
    componentLinearAccScale=<0.15 0.15 0.25>
    componentLinearVelScale=<0.15 0.15 0.25>
    overrideWorldGravity {z=-980}
   }
   Output Pose(Result=Component To Local(ComponentPose=@n11)) {layerGroup=None}
  ]}
 EventGraph {nodes [Event BlueprintUpdateAnimation()]}
]
variables []
