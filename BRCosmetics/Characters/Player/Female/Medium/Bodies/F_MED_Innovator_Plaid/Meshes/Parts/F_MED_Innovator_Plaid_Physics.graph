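# Physics asset for the F_MED_Innovator_Plaid female character part, providing simulation bodies for its dynamic hair bones.
# Capsule bodies cover back, left and right strands and three group masters, which attach to a spherical root, dyn_hair_master, with tight limits.
# Each strand constrains directly to a group master in a star layout rather than chaining, with narrow swing and twist limits.
# Side naming is crossed: _l strands attach to master_r and _r strands to master_l.
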PhysicsAsset /BRCosmetics/Characters/Player/Female/Medium/Bodies/F_MED_Innovator_Plaid/Meshes/Parts/F_MED_Innovator_Plaid_Physics
bodies {
 dyn_hair_01_bk [{shapeType=Capsule, length=1.587, radius=1.6538, center=<1.8312 0 0>, rotation=<rot 90 0 0>}]
 dyn_hair_01_l [{shapeType=Capsule, length=2.2223, radius=1.8212, center=<2.1273 0 0>, rotation=<rot 90 0 0>}]
 dyn_hair_01_r [{shapeType=Capsule, length=1.5392, radius=1.7995, center=<2.537 0 0>, rotation=<rot 90 0 0>}]
 dyn_hair_02_bk [{shapeType=Capsule, length=3.7867, radius=3.1422, center=<3.7879 0 0>, rotation=<rot 90 0 0>}]
 dyn_hair_02_l [{shapeType=Capsule, length=1.4994, radius=1.6888, center=<1.8801 0 0>, rotation=<rot 90 0 0>}]
 dyn_hair_02_r [{shapeType=Capsule, length=1.1842, radius=1.08, center=<0.8964 0 0>, rotation=<rot 90 0 0>}]
 dyn_hair_03_bk [{shapeType=Capsule, length=2.5883, radius=2.0304, center=<2.4755 0.0001 0>, rotation=<rot 90 0 0>}]
 dyn_hair_03_l [{shapeType=Capsule, length=2.7489, radius=1.3285, center=<1.7883 -0.0001 0>, rotation=<rot 90 0 0>}]
 dyn_hair_03_r [{shapeType=Capsule, length=2.3019, radius=1.5898, center=<1.2816 0 0>, rotation=<rot 90 0 0>}]
 dyn_hair_04_bk [{shapeType=Capsule, length=1.8469, radius=2.0771, center=<2.3943 0 0>, rotation=<rot 90 0 0>}]
 dyn_hair_04_l [{shapeType=Capsule, length=1.9969, radius=1.3451, center=<1.5629 0 0>, rotation=<rot 90 0 0>}]
 dyn_hair_04_r [{shapeType=Capsule, length=3.1222, radius=2.2909, center=<1.5934 0 0>, rotation=<rot 90 0 0>}]
 dyn_hair_05_l [{shapeType=Capsule, length=1.9392, radius=1.7198, center=<2.0724 0 -0.0001>, rotation=<rot 90 0 0>}]
 dyn_hair_05_r [{shapeType=Capsule, length=3.5758, radius=1.9457, center=<2.0752 0 0>, rotation=<rot 90 0 0>}]
 dyn_hair_06_l [{shapeType=Capsule, length=3.0578, radius=1.6025, center=<1.8945 0 0>, rotation=<rot 90 0 0>}]
 dyn_hair_06_r [{shapeType=Capsule, length=3.0276, radius=2.2529, center=<2.2719 0 0>, rotation=<rot 90 0 0>}]
 dyn_hair_07_l [{shapeType=Capsule, length=2.4277, radius=1.7633, center=<1.5923 0 0>, rotation=<rot 90 0 0>}]
 dyn_hair_07_r [{shapeType=Capsule, length=3.4528, radius=2.0128, center=<2.2164 0 0.0002>, rotation=<rot 90 0 0>}]
 dyn_hair_08_l [{shapeType=Capsule, length=3.7569, radius=2.2427, center=<2.8362 -0.0001 0>, rotation=<rot 90 0 0>}]
 dyn_hair_08_r [{shapeType=Capsule, length=2.5774, radius=1.9105, center=<1.6681 0 0>, rotation=<rot 90 0 0>}]
 dyn_hair_09_r [{shapeType=Capsule, length=1.3221, radius=1.8557, center=<1.6633 0.0001 0>, rotation=<rot 90 0 0>}]
 dyn_hair_10_r [{shapeType=Capsule, length=2.0776, radius=1.6305, center=<1.5144 0 -0.0001>, rotation=<rot 90 0 0>}]
 dyn_hair_master [{shapeType=Sphere, radius=1.4355}]
 dyn_hair_master_bk [{shapeType=Capsule, length=3.2951, radius=2.4463, center=<0 0 0.9325>, rotation=<rot 90 0 0>}]
 dyn_hair_master_l [{shapeType=Capsule, length=3.2951, radius=2.4463, center=<0 0 2.0724>, rotation=<rot 90 0 0>}]
 dyn_hair_master_r [{shapeType=Capsule, length=3.1686, radius=2.1262, center=<0 0 1.2808>, rotation=<rot 0 0 90>}]
}
constraints [
 {bone1Name=dyn_hair_01_l, bone2Name=dyn_hair_master_r, swing1LimitDegrees=10, swing2LimitDegrees=10, twistLimitDegrees=3}
 {bone1Name=dyn_hair_02_l, bone2Name=dyn_hair_master_r, swing1LimitDegrees=15, swing2LimitDegrees=15, twistLimitDegrees=3}
 {bone1Name=dyn_hair_03_l, bone2Name=dyn_hair_master_r, swing1LimitDegrees=15, swing2LimitDegrees=15, twistLimitDegrees=3}
 {bone1Name=dyn_hair_04_l, bone2Name=dyn_hair_master_r, swing1LimitDegrees=15, swing2LimitDegrees=15, twistLimitDegrees=3}
 {bone1Name=dyn_hair_05_l, bone2Name=dyn_hair_master_r, swing1LimitDegrees=15, swing2LimitDegrees=15, twistLimitDegrees=3}
 {bone1Name=dyn_hair_07_l, bone2Name=dyn_hair_master_r, swing1LimitDegrees=15, swing2LimitDegrees=15, twistLimitDegrees=3}
 {bone1Name=dyn_hair_08_l, bone2Name=dyn_hair_master_r, swing1LimitDegrees=15, swing2LimitDegrees=15, twistLimitDegrees=3}
 {bone1Name=dyn_hair_06_l, bone2Name=dyn_hair_master_r, swing1LimitDegrees=15, swing2LimitDegrees=15, twistLimitDegrees=3}
 {bone1Name=dyn_hair_01_bk, bone2Name=dyn_hair_master_bk, swing1LimitDegrees=15, swing2LimitDegrees=15, twistLimitDegrees=3}
 {bone1Name=dyn_hair_02_bk, bone2Name=dyn_hair_master_bk, swing1LimitDegrees=15, swing2LimitDegrees=15, twistLimitDegrees=3}
 {bone1Name=dyn_hair_03_bk, bone2Name=dyn_hair_master_bk, swing1LimitDegrees=15, swing2LimitDegrees=15, twistLimitDegrees=3}
 {bone1Name=dyn_hair_04_bk, bone2Name=dyn_hair_master_bk, swing1LimitDegrees=15, swing2LimitDegrees=15, twistLimitDegrees=3}
 {bone1Name=dyn_hair_01_r, bone2Name=dyn_hair_master_l, swing1LimitDegrees=15, swing2LimitDegrees=15, twistLimitDegrees=3}
 {bone1Name=dyn_hair_02_r, bone2Name=dyn_hair_master_l, swing1LimitDegrees=15, swing2LimitDegrees=15, twistLimitDegrees=3}
 {bone1Name=dyn_hair_03_r, bone2Name=dyn_hair_master_l, swing1LimitDegrees=15, swing2LimitDegrees=15, twistLimitDegrees=3}
 {bone1Name=dyn_hair_04_r, bone2Name=dyn_hair_master_l, swing1LimitDegrees=6, swing2LimitDegrees=6, twistLimitDegrees=3}
 {bone1Name=dyn_hair_05_r, bone2Name=dyn_hair_master_l, swing1LimitDegrees=6, swing2LimitDegrees=6, twistLimitDegrees=3}
 {bone1Name=dyn_hair_06_r, bone2Name=dyn_hair_master_l, swing1LimitDegrees=15, swing2LimitDegrees=15, twistLimitDegrees=3}
 {bone1Name=dyn_hair_07_r, bone2Name=dyn_hair_master_l, swing1LimitDegrees=6, swing2LimitDegrees=6, twistLimitDegrees=3}
 {bone1Name=dyn_hair_08_r, bone2Name=dyn_hair_master_l, swing1LimitDegrees=6, swing2LimitDegrees=6, twistLimitDegrees=3}
 {bone1Name=dyn_hair_09_r, bone2Name=dyn_hair_master_l, swing1LimitDegrees=6, swing2LimitDegrees=6, twistLimitDegrees=3}
 {bone1Name=dyn_hair_10_r, bone2Name=dyn_hair_master_l, swing1LimitDegrees=6, swing2LimitDegrees=6, twistLimitDegrees=3}
 {bone1Name=dyn_hair_master_r, bone2Name=dyn_hair_master, swing1LimitDegrees=3, swing2LimitDegrees=3, twistLimitDegrees=2}
 {bone1Name=dyn_hair_master_bk, bone2Name=dyn_hair_master, swing1LimitDegrees=3, swing2LimitDegrees=3, twistLimitDegrees=2}
 {bone1Name=dyn_hair_master_l, bone2Name=dyn_hair_master, swing1LimitDegrees=3, swing2LimitDegrees=3, twistLimitDegrees=2}
]
